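# Conversation database for the Serpent Coil NPC, defining its greeting and top-level dialogue branches.
# A single entry point leads to a speech node with a default line and contextual lines prioritized by curve rows.
# Context checks cover first-ever conversation, specific player outfits via character ID, and per-match conversation count.
# The speech node then links by entry tag to shared quests, basic services, a Ch5S1 quest entry and Serpent Coil's own dialogue.
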ConversationDatabase /NPCLibrary/NPCs/SerpentCoil/NPCLibrary_Conversation_SerpentCoil
entry_points [{tag=Conversation.Entry.SerpentCoil, destinations [#1]}]
nodes [
 ConversationEntryPointNode_7 {
  id=#1
  class=ConversationEntryPointNode
  outputs [#2]
  properties {entryTag {tagName=Conversation.Entry.SerpentCoil}}
 }
 Speech_C_4 {
  id=#2
  class=Speech_C
  outputs [#3, #4, #5, #6]
  properties {
   nodeName=""
   generalConfig {
    defaultMessage="I'm not one of you Underground foolsssss, but I can help you out now and again."
    contextualMessages [
     {
      message="So you're the one ssssstanding against The Society. Unwise, but brave."
      contextRequirements [{requirement {class=FortControllerRequirement_IsFirstEverConversationWithNPC}}]
      priority {curve {rowName=Tandem.Speech.Priority.FirstEverInteraction}}
     }
     {
      message="You're playing a dangerousssss game, Hope. You'd better be successful."
      contextRequirements [{requirement {
         class=FortControllerRequirement_HasCID
         properties {allowedCIds [&Character_VoidRedemption_Rebel]}
        }}]
      priority {curve {rowName=Tandem.Speech.Priority.OutfitReaction}}
     }
     {
      message="I have no interest in joining your Society, so sssssstop asking me."
      contextRequirements [{requirement {
         class=FortControllerRequirement_HasCID
         properties {allowedCIds [/BRCosmetics/Athena/Items/Cosmetics/Characters/Character_DiamondHeart_Chic]}
        }}]
      priority {curve {rowName=Tandem.Speech.Priority.OutfitReaction}}
     }
     {
      message=I know you're hunting for ancient ssssecrets. But what might they be...?
      contextRequirements [{requirement {class=FortControllerRequirement_HasCID, properties {allowedCIds [/BRCosmetics/Athena/Items/Cosmetics/Characters/Character_MadameMoth_Posh]}}}]
      priority {curve {rowName=Tandem.Speech.Priority.OutfitReaction}}
     }
     {
      message="I'm not one of you Underground foolsssss, but I can help you out now and again."
      contextRequirements [{requirement {class=FortControllerRequirement_CheckNPCMatchConversationCount, properties {nPCUniqueIdTag {tagName=AISpawnerData.Type.Tandem.SerpentCoil}}}}]
      priority {curve {rowName=None}}
     }
    ]
   }
  }
 }
 ConversationLinkNode_3 {
  id=#6
  class=ConversationLinkNode
  properties {remoteEntryTag {tagName=Conversation.Entry.SerpentCoil.Dialogue}}
 }
 ConversationLinkNode_2 {id=#5, class=ConversationLinkNode, properties {remoteEntryTag {tagName=Conversation.Entry.Ch5S1.D3.Q8}}}
 ConversationLinkNode_0 {id=#4, class=ConversationLinkNode, properties {remoteEntryTag {tagName=Conversation.Entry.BasicServices}}}
 ConversationLinkNode_1 {id=#3, class=ConversationLinkNode, properties {remoteEntryTag {tagName=Conversation.Entry.NPCLibrary.Quests}}}
]
