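# Conversation database for the Serpent Coil NPC's dialogue, entered through its Conversation.Entry.SerpentCoil.Dialogue tag.
# An "Ask about..." hub speech branches to the snake topic, the society topic, or a Go Back node.
# The snake topic chains into a second speech line, and both topics end on a shared ReturnToStart node labeled Go Back.
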ConversationDatabase /NPCLibrary/NPCs/SerpentCoil/NPCLibrary_Conversation_SerpentCoil_Dialogue
entry_points [{tag=Conversation.Entry.SerpentCoil.Dialogue, destinations [#1]}]
nodes [
 ConversationEntryPointNode_0 {
  id=#1
  class=ConversationEntryPointNode
  outputs [#2]
  properties {entryTag {tagName=Conversation.Entry.SerpentCoil.Dialogue}}
 }
 Speech_C_0 {
  id=#2
  class=Speech_C
  outputs [#3, #4, #5]
  properties {nodeName="", softTaskIcon=&T_UI_NPC_question, generalConfig {defaultMessage=...}}
  sub_nodes [ConversationChoiceNode_5 {class=ConversationChoiceNode, properties {defaultChoiceDisplayText=Ask about...}}]
 }
 Back_C_3 {
  id=#5
  class=Back_C
  sub_nodes [ConversationChoiceNode_23 {class=ConversationChoiceNode, properties {defaultChoiceDisplayText=Go Back}}]
 }
 Speech_C_4 {
  id=#4
  class=Speech_C
  outputs [#6]
  properties {
   nodeName=""
   softTaskIcon=&T_UI_NPC_question
   generalConfig {defaultMessage=The Sssssociety is arrogant. The only arrogance I find charming is mine.}
  }
  sub_nodes [ConversationChoiceNode_19 {class=ConversationChoiceNode, properties {defaultChoiceDisplayText=The society}}]
 }
 ReturnToStart_C_0 {
  id=#6
  class=ReturnToStart_C
  properties {softTaskIcon=&T_UI_NPC_Back}
  sub_nodes [ConversationChoiceNode_11 {class=ConversationChoiceNode, properties {defaultChoiceDisplayText=Go Back}}]
 }
 Speech_C_14 {
  id=#3
  class=Speech_C
  outputs [#7]
  properties {nodeName="", softTaskIcon=&T_UI_NPC_question, generalConfig {defaultMessage=Don't ask about my ssssnake friend.}}
  sub_nodes [ConversationChoiceNode_18 {class=ConversationChoiceNode, properties {defaultChoiceDisplayText=The snake}}]
 }
 Speech_C_12 {
  id=#7
  class=Speech_C
  outputs [#6]
  properties {
   nodeName=""
   softTaskIcon=&T_UI_NPC_question
   generalConfig {defaultMessage="He is just a pet, and definitely not possessssssing this body."}
  }
 }
]
